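; Gameplay ability that manages Zero Point dash charges for the Silky Wolf Haunted POI token, extending a parent ability.
; Its only active logic overrides the dash-charge use event, calls the parent, then reads the remaining charges.
; When none remain, it removes all stacks of the Zero Point dash GameplayEffect from the ability system component so the Crashpad UI element works.
; Activate and end events are disabled stubs.
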(event Ability|EventActivateAbility
  (|CallParentFunction) ; This node is disabled and will not be called.\nDrag off pins to build functionality.
)

(event Ability|EventOnEndAbility
  (params (bWasCancelled :type Boolean))
  (|CallParentFunction bWasCancelled) ; This node is disabled and will not be called.\nDrag off pins to build functionality.
)

(event EventUseDashCharge
  (params (Number%20of%20Charges :type Integer))
  ; Remove Gameplay Effect so that Crashpad UI element works
  (|CallParentFunction Number%20of%20Charges)
  (bind _remaining_charges (CallFunction|GetRemainingCharges))
  (if (<= _remaining_charges 0)
    (GameplayEffects|RemoveActiveGameplayEffectbySourceEffect (Variables|Default|GetAbility%20System%20Component) "/DreamFog/Gameplay/Dash/GE_HauntedPOI_Token_SilkyWolf_ZeroPoint_DashAbility.GE_HauntedPOI_Token_SilkyWolf_ZeroPoint_DashAbility_C" 0 -1)))
